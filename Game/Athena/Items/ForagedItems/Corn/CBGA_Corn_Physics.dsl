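; Physics-driven corn foraged item for Athena that a pawn can consume or gather.
; On BeginPlay it tips the mesh by a pitch offset with random yaw, and its hit event spawns optional ground FX and sound.
; First interaction requires health not at max plus parent approval; second follows the GatherEnabled data row.
; The gather sound plays inside that permission check rather than on an actual gather.
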(event EventBeginPlay
  ;[1 1 1] Set Corn to the side on begin play
  (|CallParentFunction)
  (Transformation|SetWorldRotation (Variables|Default|GetStaticMesh) (Math|Rotator|MakeRotator 0.0 (Variables|Default|GetPitchOffset) (Math|Random|RandomFloatinRange 0.0 359.0)) false false))

(event EventPlayHitFX
  (params (Location :type Vector :default "0, 0, 0") (Rotation :type Rotator :default "0, 0, 0"))
  (bind (_returnvalue_roll _returnvalue_pitch _returnvalue_yaw) (Math|Rotator|BreakRotator (Transformation|GetActorRotation)))
  (bind _hitgroundfx (Variables|Gravity|GetHitgroundFX))
  (bind _hitgroundsound (Variables|Gravity|GetHitgroundSound))
  ;[1 1 1] Play Sounds and FX if valid
  (Utilities|IsValid _hitgroundfx
    (:Is%20Valid
      (Effects|SpawnEmitteratLocation _hitgroundfx (Transformation|GetActorLocation) (Math|Rotator|MakeRotator _returnvalue_roll 9e+01 _returnvalue_yaw) "1.000000,1.000000,1.000000" true "None" true))
    (:Is%20Not%20Valid
      _))
  (Utilities|IsValid _hitgroundsound
    (:Is%20Valid
      (Audio|PlaySoundatLocation _hitgroundsound (Transformation|GetActorLocation)))))

(fn OnInteractAllowedToConsume
  (params (InteractingPawn :type Fort%20Pawn%20Object%20Reference) (InteractionBeingAttempted :type EInteractionBeingAttempted%20Enum :default "FirstInteraction"))
  (bind _row_gatherenabled (Variables|Gather|GetRow_GatherEnabled))
  (bind _returnvalue (Class|AthenaFunctionLibrary|RowtoBool _row_gatherenabled))
  (Utilities|IsValid InteractingPawn
    (:Is%20Valid
      ;[1 1 1] If second interact allow gather if enabled
      (switch Utilities|FlowControl|Switch|SwitchonEInteractionBeingAttempted InteractionBeingAttempted
        (:FirstInteraction
          (return (and (not (Math|Float|NearlyEqual(Float) (FortPawn|GetHealth InteractingPawn) (FortPawn|GetMaxHealth InteractingPawn))) (|CallParentFunction InteractingPawn "FirstInteraction" "IT_Simple"))))
        (:SecondInteraction
          (if _returnvalue
            (Audio|PlaySoundatLocation (Variables|Default|GetGatherSound) (Transformation|GetActorLocation) "0, 0, 0" 0.4))
          (return _returnvalue))))
    (:Is%20Not%20Valid
      (return false))))

(fn ConstructionScript
  (|CallParentFunction))
